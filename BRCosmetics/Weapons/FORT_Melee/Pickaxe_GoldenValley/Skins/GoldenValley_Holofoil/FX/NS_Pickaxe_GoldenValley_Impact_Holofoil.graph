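# Niagara impact effect for the Holofoil skin of the GoldenValley pickaxe, listing ten emitters including flare, smoke, spikes, sparks, confetti and petals.
# Only Petals has detail here: a slightly delayed burst flung outward from a tiny shape, slowed by curve-driven drag, then drifting under gravity and curl noise.
# An age-driven sine on sprite width fakes tumbling alongside sprite rotation.
# Front-end inputs reduce petal count, size and speed.
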NiagaraSystem /BRCosmetics/Weapons/FORT_Melee/Pickaxe_GoldenValley/Skins/GoldenValley_Holofoil/FX/NS_Pickaxe_GoldenValley_Impact_Holofoil
dependencies {
 usedDataInterfaces [&NiagaraDataInterfaceCurve]
 usedDynamicInputs [&MakeCustomFloatFromBool]
 usedModules [
  &self:SystemSpawnScript.SystemScriptSource.SystemScriptGraph.NiagaraNodeAssignment_0.SetVariables_#1
  &self:SystemSpawnScript.SystemScriptSource.SystemScriptGraph.NiagaraNodeAssignment_1.SetVariables_#2
  &self:SystemSpawnScript.SystemScriptSource.SystemScriptGraph.NiagaraNodeAssignment_2.SetVariables_#3
  &SystemState
 ]
}
emitters {
 ConfettiLWE {rendererClasses [], renderers []}
 DarkBG {rendererClasses [], renderers []}
 FlareAndLight {rendererClasses [], renderers []}
 Petals {
  emitterUpdateScript [
   &EmitterState(Life Cycle Mode=::NewEnumerator1, Loop Behavior=::NewEnumerator1, Loop Delay=0.025, Loop Duration=0.5, UseLoopDelay=true)
   &SpawnBurst_Instantaneous(Spawn Count=200, Use Spawn Probability=false, UseLoopCountLimit=false) {enabled=false, module=SpawnBurst_Instantaneous001}
   &SpawnBurst_Instantaneous(
    Spawn Count=&MakeCustomIntFromBool(
     Bool=$User.bIsFrontEndPreview
     True INT=&RandomRangeInt(Minimum=5, Maximum=10)
     False INT=&RandomRangeInt(Minimum=10, Maximum=15)
    )
    Use Spawn Probability=false
    UseLoopCountLimit=false
   )
  ]
  particleSpawnScript [
   &InitializeParticle(
    Lifetime=&RandomRangeFloat(Minimum=1, Maximum=1.5)
    Sprite Rotation Mode=::NewEnumerator2
    Sprite Size=&Multiply_Vector2D_ByFloat(
     Float=&RandomRangeFloat(Minimum=0.4, Maximum=0.6)
     Vector2D=&Multiply_Vector2D_ByFloat(Float=&FN_FrontEndFloatModifier(Gameplay=1, Front End=0.75), Vector2D=<26 10>)
    )
    Sprite Size Mode=::NewEnumerator4
    UsePositionOffset=false
    Write Material Random=true
   ) {module=InitializeParticle001}
   &ShapeLocation(
    Additional Quaternion Rotation=true
    Additional Yaw / Pitch / Roll=<-90 0 0>
    Hemisphere Distribution=<0.5 1>
    MaskSpawnGroup=false
    Override Seed=false
    Rotation Angle Type:ENiagara_AngleInput=::NewEnumerator0
    Rotation Mode=::NewEnumerator3
    Sphere Radius=0.01
    UseOwnerScale=true
   )
   &AddVelocity(
    Rotation Coordinate Space=::Simulation
    Velocity=&Multiply_VectorByFloat(Vector=$Particles.ShapeLocation.ShapeNormal, Float=&RandomRangeFloat(Minimum=1000, Maximum=1500))
    Velocity Speed Scale=&FN_FrontEndFloatModifier(Gameplay=1, Front End=0.8)
   )
   /SetVariables(Particles.stableGravity:NiagaraFloat=&UniformRangedFloat(Minimum=-150, Maximum=-200)) {isSetParameters=true}
   /SetVariables(Particles.stableRotationRate:NiagaraFloat=&UniformRangedFloat(Minimum=50, Maximum=200)) {isSetParameters=true}
  ]
  particleUpdateScript [
   &ParticleState()
   &GravityForce(Gravity=&MakeVector(X=0, Y=0, Z=$Particles.Initial.stableGravity))
   &Drag(
    Drag=&Multiply_Float(
     A=&UniformRangedFloat(Minimum=1.2, Maximum=1.5)
     B=&FloatFromCurve(
      FloatCurve=DataInterface() {propertyValues {Curve {keys [
          {arriveTangent=-5.3551, interpMode=RCIM_Cubic, leaveTangent=-5.3551, tangentMode=RCTM_User, time=0, value=0}
          {arriveTangent=-138.8873, interpMode=RCIM_Cubic, leaveTangent=-138.8873, tangentMode=RCTM_User, time=0.05, value=7}
          <at 0.2 0.5>
         ]}}}
      Scale Curve=4
     )
    )
    Use Linear Drag=true
    Use Rotational Drag=false
   )
   /Niagara/Modules/Update/Forces/CurlNoiseForce(Noise Strength=300, Pan Noise=false)
   &SolveForcesAndVelocity(Clamp Velocity=false, Limit Acceleration=false)
   &SpriteRotationRate(Rotation Rate=$Particles.Initial.stableRotationRate)
   &ScaleSpriteSize(Scale Factor=&Lerp_Vector2(
     A=&Multiply_Vector2(
      A=&MakeVector2D(
       X=&Multiply_Float(
        A=1
        B=&Sine(
         Normalized Angle=$Emitter.Age
         Period=&Multiply_Float(A=0.7, B=&Add_Float(A=$Particles.Initial.MaterialRandom, B=0.65))
         Scale=1
         Bias=0
        )
       )
       Y=1
      )
      B=&Vector2DFromFloat(Value=&FloatFromCurve(FloatCurve=DataInterface() {propertyValues {Curve {keys [<at 0 1.1>,<at 0.9 1.1>,<at 1 0>]}}}))
     )
     B=<1 1>
     Alpha=&FloatFromCurve(FloatCurve=DataInterface() {propertyValues {Curve {keys [<at 0.0574 1>,<at 0.1545 0>]}}})
    ))
  ]
  rendererClasses [&NiagaraSpriteRendererProperties]
  renderers [
   {rendererClass=&NiagaraSpriteRendererProperties}
   {rendererClass=&NiagaraSpriteRendererProperties, rendererIndex=1}
  ]
 }
 SmokePuff {rendererClasses [], renderers []}
 Sparks {rendererClasses [], renderers []}
 SpikeGreen {rendererClasses [], renderers []}
 SpikePink {rendererClasses [], renderers []}
 StarSpike {rendererClasses [], renderers []}
 Wave {rendererClasses [], renderers []}
}
summary {
 systemName=NS_Pickaxe_GoldenValley_Impact_Holofoil
 emitters [
  {emitterName=FlareAndLight, rendererClasses []}
  {emitterName=SmokePuff, rendererClasses []}
  {emitterName=SpikePink, rendererClasses []}
  {emitterName=StarSpike, rendererClasses []}
  {emitterName=DarkBG, rendererClasses []}
  {emitterName=SpikeGreen, rendererClasses []}
  {emitterName=Wave, rendererClasses []}
  {emitterName=Sparks, rendererClasses []}
  {emitterName=ConfettiLWE, rendererClasses []}
  {emitterName=Petals, rendererClasses [&NiagaraSpriteRendererProperties]}
 ]
 userVariables [
  User.bIsFrontEnd {type=NiagaraBool}
  User.bIsFrontEndPreview {type=NiagaraBool}
  User.Variant {type=NiagaraFloat, defaultValue=0, description=""}
 ]
}
